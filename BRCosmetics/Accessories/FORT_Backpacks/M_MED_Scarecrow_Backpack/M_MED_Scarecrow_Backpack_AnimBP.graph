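# Anim Blueprint for the Scarecrow backpack cosmetic, adding procedural secondary motion to its body, petals, feathers, nest and eggs.
# The input pose gets a small bodyOffset nudge, then passes through constrained Anim Dynamics for body swing and four petal bones.
# Trail controllers with planar limits move three feathers, then spring-driven Anim Dynamics jiggle the nest and two eggs.
# A separate feather Anim Dynamics chain is left unconnected, and the event graph has no logic.
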AnimBlueprint /BRCosmetics/Accessories/FORT_Backpacks/M_MED_Scarecrow_Backpack/M_MED_Scarecrow_Backpack_AnimBP
target_skeleton=&/M_MED_Scarecrow_Backpack_Skeleton
graphs [
 AnimGraph {nodes [
   n1 = Anim Dynamics - Bone: featherA1(Alpha=1, ExternalForce="0,0,0") {
    angularDampingOverride=0.999
    bGravityOverrideInSimSpace=true
    bOverrideAngularDamping=true
    gravityScale=0.01
    alphaBoolBlend {blendOption=Linear}
    boundBone {boneName=featherA1}
    componentAppliedLinearAccClamp=<50 50 50>
    componentLinearAccScale=<1 1 1>
    componentLinearVelScale=<1 1 1>
    physicsBodyDefinitions [{
      sphereCollisionRadius=0
      boundBone {boneName=featherA1}
      boxExtents=<0.1 0.1 0.1>
      constraintSetup {angularLimitsMax=<0 0 0>, angularLimitsMin=<0 0 0>, angularTarget=<0 0 0>}
      localJointOffset=<0 0 -1>
     }]
   }
   n2 = Anim Dynamics - Bone: featherB1(ComponentPose=@n1, Alpha=1, ExternalForce="0,0,0") {
    angularDampingOverride=0.999
    bGravityOverrideInSimSpace=true
    bOverrideAngularDamping=true
    gravityScale=0.01
    alphaBoolBlend {blendOption=Linear}
    boundBone {boneName=featherB1}
    componentAppliedLinearAccClamp=<50 50 50>
    componentLinearAccScale=<1 1 1>
    componentLinearVelScale=<1 1 1>
    physicsBodyDefinitions [{
      sphereCollisionRadius=0
      boundBone {boneName=featherB1}
      boxExtents=<0.1 0.1 0.1>
      constraintSetup {angularLimitsMax=<0 0 0>, angularLimitsMin=<0 0 0>, angularTarget=<0 0 0>}
      localJointOffset=<0 0 -1>
     }]
   }
   Anim Dynamics - Bone: featherC1(ComponentPose=@n2, Alpha=1, ExternalForce="0,0,0") {
    angularDampingOverride=0.999
    bGravityOverrideInSimSpace=true
    bOverrideAngularDamping=true
    gravityScale=0.01
    alphaBoolBlend {blendOption=Linear}
    boundBone {boneName=featherC1}
    componentAppliedLinearAccClamp=<50 50 50>
    componentLinearAccScale=<1 1 1>
    componentLinearVelScale=<1 1 1>
    physicsBodyDefinitions [{
      sphereCollisionRadius=0
      boundBone {boneName=featherC1}
      boxExtents=<0.1 0.1 0.1>
      constraintSetup {angularLimitsMax=<0 0 0>, angularLimitsMin=<0 0 0>, angularTarget=<0 0 0>}
      localJointOffset=<1 0 -1>
     }]
   }
   AnimGraphNode_SubInput_0 = Input Pose()
   n3 = "Transform (Modify) Bone - Bone: bodyOffset"!(ComponentPose=Local To Component(LocalPose=@AnimGraphNode_SubInput_0), Alpha=1) -> (Pose) {
    rotationSpace=BCS_WorldSpace
    translationMode=BMM_Additive
    alphaBoolBlend {blendOption=Linear}
    boneToModify {boneName=bodyOffset}
    translation {y=0.2}
   }
   n4 = Anim Dynamics - Bone: dyn_bodySwing(ComponentPose=@n3, Alpha=1, ExternalForce="0,0,0") {
    angularDampingOverride=0.999
    angularSpringConstant=500
    bAngularSpring=true
    bGravityOverrideInSimSpace=true
    bOverrideAngularDamping=true
    gravityScale=0.3
    alphaBoolBlend {blendOption=Linear}
    boundBone {boneName=dyn_bodySwing}
    physicsBodyDefinitions [{
      sphereCollisionRadius=0
      boundBone {boneName=dyn_bodySwing}
      boxExtents=<5 15 15>
      constraintSetup {angularLimitsMax=<0 0 12>, angularLimitsMin=<0 0 0>, angularTarget=<0 0 0>}
      localJointOffset=<-10 0 0>
     }]
   }
   n5 = Anim Dynamics - Bone: dyn_petalsAbase(ComponentPose=@n4, Alpha=1, ExternalForce="0,0,0") {
    angularDampingOverride=0.999
    bGravityOverrideInSimSpace=true
    bOverrideAngularDamping=true
    gravityScale=0.01
    alphaBoolBlend {blendOption=Linear}
    boundBone {boneName=dyn_petalsAbase}
    componentAppliedLinearAccClamp=<3000 3000 3000>
    componentLinearAccScale=<1.1 1.1 1.1>
    componentLinearVelScale=<1.1 1.1 1.1>
    physicsBodyDefinitions [{
      sphereCollisionRadius=0
      boundBone {boneName=dyn_petalsAbase}
      boxExtents=<5 5 5>
      constraintSetup {angularLimitsMax=<0 17.1975 5>, angularLimitsMin=<0 -27.5159 -5>, angularTarget=<0 0 0>}
      localJointOffset=<-3 0 0>
     }]
   }
   n6 = Anim Dynamics - Bone: dyn_petalsBbase(ComponentPose=@n5, Alpha=1, ExternalForce="0,0,0") {
    angularDampingOverride=0.999
    bGravityOverrideInSimSpace=true
    bOverrideAngularDamping=true
    gravityScale=0.01
    alphaBoolBlend {blendOption=Linear}
    boundBone {boneName=dyn_petalsBbase}
    componentAppliedLinearAccClamp=<3000 3000 3000>
    componentLinearAccScale=<1.1 1.1 1.1>
    componentLinearVelScale=<1.1 1.1 1.1>
    physicsBodyDefinitions [{
      sphereCollisionRadius=0
      boundBone {boneName=dyn_petalsBbase}
      boxExtents=<5 5 5>
      constraintSetup {angularLimitsMax=<0 5 30.9554>, angularLimitsMin=<0 -5 -30.9554>, angularTarget=<0 0 0>}
      localJointOffset=<-3 0 0>
     }]
   }
   n7 = Anim Dynamics - Bone: dyn_petalsCbase(ComponentPose=@n6, Alpha=1, ExternalForce="0,0,0") {
    angularDampingOverride=0.999
    bGravityOverrideInSimSpace=true
    bOverrideAngularDamping=true
    gravityScale=0.01
    alphaBoolBlend {blendOption=Linear}
    boundBone {boneName=dyn_petalsCbase}
    componentAppliedLinearAccClamp=<3000 3000 3000>
    componentLinearAccScale=<1.1 1.1 1.1>
    componentLinearVelScale=<1.1 1.1 1.1>
    physicsBodyDefinitions [{
      sphereCollisionRadius=0
      boundBone {boneName=dyn_petalsCbase}
      boxExtents=<5 5 5>
      constraintSetup {angularLimitsMax=<0 30 5>, angularLimitsMin=<0 -30 -5>, angularTarget=<0 0 0>}
      localJointOffset=<0 -3 0>
     }]
   }
   n8 = Anim Dynamics - Bone: dyn_petalsDtmp(ComponentPose=@n7, Alpha=1, ExternalForce="0,0,0") {
    angularDampingOverride=0.999
    bGravityOverrideInSimSpace=true
    bOverrideAngularDamping=true
    gravityScale=0.01
    alphaBoolBlend {blendOption=Linear}
    boundBone {boneName=dyn_petalsDtmp}
    componentAppliedLinearAccClamp=<3000 3000 3000>
    componentLinearAccScale=<1.1 1.1 1.1>
    componentLinearVelScale=<1.1 1.1 1.1>
    physicsBodyDefinitions [{
      sphereCollisionRadius=0
      boundBone {boneName=dyn_petalsDtmp}
      boxExtents=<5 5 5>
      constraintSetup {twistAxis=AxisY, angularLimitsMax=<0 5 27.5159>, angularLimitsMin=<0 -5 -30.9554>, angularTarget=<0 0 0>}
      localJointOffset=<-3 0 0>
     }]
   }
   n9 = Trail controller - Bone: featherA1(ComponentPose=@n8, Alpha=1) {
    bUsePlanarLimit=true
    relaxationSpeedScale=6
    planarLimits [
     {drivingBone {boneName=featherAbase}, planeTransform {location=<2 0 0>, rotation=<rot 0 114.9999 -102>, scale=<0.2 0.2 0.2>}}
     {drivingBone {boneName=featherAbase}, planeTransform {location=<2 0 1>, rotation=<rot 0 0 -180>, scale=<0.2 0.2 0.2>}}
    ]
    rotationLimits [{limitMax=<180 180 180>}, {limitMax=<180 180 180>}]
    rotationOffsets [<0 0 0>, <0 0 0>]
    trailBone {boneName=featherA1}
   }
   n10 = Trail controller - Bone: featherB1(ComponentPose=@n9, Alpha=1) {
    bUsePlanarLimit=true
    relaxationSpeedScale=6
    planarLimits [
     {drivingBone {boneName=featherBbase}, planeTransform {location=<4 0 0>, rotation=<rot 0 99 -66.0001>, scale=<0.2 0.2 0.2>}}
     {drivingBone {boneName=dyn_bodySwing}, planeTransform {location=<-13 -10 -10>, rotation=<rot 90 0 0>, scale=<0.3 0.3 0.3>}}
     {drivingBone {boneName=dyn_bodySwing}, planeTransform {location=<-13 -10 -5>, rotation=<rot -59.9999 174.2893 144.9999>, scale=<0.2 0.2 0.2>}}
    ]
    rotationLimits [{limitMax=<180 180 180>}, {limitMax=<180 180 180>}]
    rotationOffsets [<0 0 0>, <0 0 0>]
    trailBone {boneName=featherB1}
   }
   n11 = Trail controller - Bone: featherC1(ComponentPose=@n10, Alpha=1) {
    bUsePlanarLimit=true
    relaxationSpeedScale=6
    planarLimits [
     {drivingBone {boneName=featherCbase}, planeTransform {location=<4.5 0 0>, rotation=<rot 0 114.9998 -66.0003>, scale=<0.2 0.2 0.2>}}
     {drivingBone {boneName=dyn_bodySwing}, planeTransform {location=<-13 -10 -10>, rotation=<rot 90 0 0>, scale=<0.3 0.3 0.3>}}
     {drivingBone {boneName=dyn_bodySwing}, planeTransform {location=<-13 -10 -5>, rotation=<rot -59.9999 174.9999 155>, scale=<0.2 0.2 0.2>}}
    ]
    rotationLimits [{limitMax=<180 180 180>}, {limitMax=<180 180 180>}]
    rotationOffsets [<0 0 0>, <0 0 0>]
    trailBone {boneName=featherC1}
   }
   n12 = Anim Dynamics - Bone: dyn_nest(ComponentPose=@n11, Alpha=1, ExternalForce="0,0,0") {
    angularDampingOverride=0.999
    angularSpringConstant=500
    bAngularSpring=true
    bGravityOverrideInSimSpace=true
    bOverrideAngularDamping=true
    gravityScale=0.01
    alphaBoolBlend {blendOption=Linear}
    boundBone {boneName=dyn_nest}
    physicsBodyDefinitions [{
      sphereCollisionRadius=0
      boundBone {boneName=dyn_nest}
      boxExtents=<1 10 10>
      constraintSetup {angularLimitsMax=<0 3 3>, angularLimitsMin=<0 -3 -3>, angularTarget=<0 0 0>}
      localJointOffset=<-5 0 0>
     }]
   }
   n13 = Anim Dynamics - Bone: dyn_egg1(ComponentPose=@n12, Alpha=1, ExternalForce="0,0,0") {
    angularDampingOverride=0.999
    angularSpringConstant=500
    bAngularSpring=true
    bGravityOverrideInSimSpace=true
    bOverrideAngularDamping=true
    gravityScale=0.01
    alphaBoolBlend {blendOption=Linear}
    boundBone {boneName=dyn_egg1}
    physicsBodyDefinitions [{
      sphereCollisionRadius=0
      boundBone {boneName=dyn_egg1}
      boxExtents=<1 9 8>
      constraintSetup {angularLimitsMax=<0 3 3>, angularLimitsMin=<0 -3 -3>, angularTarget=<0 0 0>}
      localJointOffset=<-5 0 0>
     }]
   }
   n14 = Anim Dynamics - Bone: dyn_egg2(ComponentPose=@n13, Alpha=1, ExternalForce="0,0,0") {
    angularDampingOverride=0.999
    angularSpringConstant=500
    bAngularSpring=true
    bGravityOverrideInSimSpace=true
    bOverrideAngularDamping=true
    gravityScale=0.01
    alphaBoolBlend {blendOption=Linear}
    boundBone {boneName=dyn_egg2}
    physicsBodyDefinitions [{
      sphereCollisionRadius=0
      boundBone {boneName=dyn_egg2}
      boxExtents=<1 12 7>
      constraintSetup {angularLimitsMax=<0 3 3>, angularLimitsMin=<0 -3 -3>, angularTarget=<0 0 0>}
      localJointOffset=<-5 0 0>
     }]
   }
   Output Pose(Result=Component To Local(ComponentPose=@n14)) {layerGroup=None}
  ]}
 EventGraph {nodes [Event BlueprintUpdateAnimation()]}
]
variables []
